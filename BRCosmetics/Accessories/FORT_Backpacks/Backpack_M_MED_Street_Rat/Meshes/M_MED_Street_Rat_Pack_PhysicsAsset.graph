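# Physics asset for the Street Rat backpack cosmetic, defining collision and joints for the pack's dangling tassel.
# One capsule body covers the pack and three capsules follow the tassel bones in a chain.
# The first tassel link has narrow swing limits against the pack, while the lower two links swing more widely with tighter twist, so the tassel bends mostly along its length.
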PhysicsAsset /BRCosmetics/Accessories/FORT_Backpacks/Backpack_M_MED_Street_Rat/Meshes/M_MED_Street_Rat_Pack_PhysicsAsset
bodies {
 dyn_tassel_01 [{shapeType=Capsule, length=2.1894, radius=1.6952, center=<3.5604 -0.0089 0.0052>, rotation=<rot 90 0 0>}]
 dyn_tassel_02 [{shapeType=Capsule, length=4.8158, radius=2.3658, center=<4.5983 -0.3803 -0.0745>, rotation=<rot 90 0 0>}]
 dyn_tassel_03 [{shapeType=Capsule, length=4.9851, radius=3.1557, center=<3.6223 -0.4727 -0.5803>, rotation=<rot 90 0 0>}]
 pack [{shapeType=Capsule, length=0.1664, radius=11.1228, center=<-0.0153 -10.2558 0.4874>, rotation=<rot 0.3025 -90 -1.7606>}]
}
constraints [
 {bone1Name=dyn_tassel_01, bone2Name=pack, swing1LimitDegrees=9, swing2LimitDegrees=9, twistLimitDegrees=15}
 {bone1Name=dyn_tassel_02, bone2Name=dyn_tassel_01, swing1LimitDegrees=45, swing2LimitDegrees=45, twistLimitDegrees=5}
 {bone1Name=dyn_tassel_03, bone2Name=dyn_tassel_02, swing1LimitDegrees=45, swing2LimitDegrees=45, twistLimitDegrees=5}
]
